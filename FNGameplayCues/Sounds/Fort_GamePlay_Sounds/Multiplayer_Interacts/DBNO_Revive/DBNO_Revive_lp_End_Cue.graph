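# Sound cue for the end of the down-but-not-out revive loop, playing a player "unjoin" sound.
# A Modulator feeds a single wave player for the one unjoin sample.
# It plays in the in-game SFX class with the shared player interact attenuation.
# An overridden concurrency limits voices per owner, stops the oldest and applies a retrigger time.
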SoundCue /FNGameplayCues/Sounds/Fort_GamePlay_Sounds/Multiplayer_Interacts/DBNO_Revive/DBNO_Revive_lp_End_Cue
root=Modulator_0
nodes [
 {id=Modulator_0, children [WavePlayer_1]}
 {
  id=WavePlayer_1
  properties {soundWaveAssetPtr=/FNGameplayCues/Sounds/Fort_GamePlay_Sounds/Multiplayer_Interacts/DBNO_Revive/wav/DBNO_Revive_PlayerUnjoin_01}
 }
]
properties {
 attenuationSettings=/FNGameplayCues/Sounds/Fort_GamePlay_Sounds/Multiplayer_Interacts/Generic_Interact/Interact_Player_Att
 bOverrideConcurrency=true
 maxDistance=1300
 soundClassObject=&Fort_SFX_InGame
 volumeMultiplier=1
 concurrencyOverrides {bLimitToOwner=true, maxCount=3, resolutionRule=StopOldest, retriggerTime=0.3, voiceStealReleaseTime=0.1, volumeScale=0.8}
}
