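# DBNO_Revive_lp_Start_Cue is a sound cue for a DBNO revive multiplayer interaction, apparently its start sound.
# A single wave player plays the PlayerJoin wave through a modulator that adds slight random pitch and volume boosts.
# It uses the revive attenuation asset and the in-game SFX sound class.
# An owner-limited concurrency override caps simultaneous instances and sets a short retrigger time.
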SoundCue /FNGameplayCues/Sounds/Fort_GamePlay_Sounds/Multiplayer_Interacts/DBNO_Revive/DBNO_Revive_lp_Start_Cue
root=Modulator_1
nodes [
 {id=Modulator_1, children [WavePlayer_1], properties {pitchMax=1.1, pitchMin=1, volumeMax=1.3, volumeMin=1.1}}
 {
  id=WavePlayer_1
  properties {soundWaveAssetPtr=/FNGameplayCues/Sounds/Fort_GamePlay_Sounds/Multiplayer_Interacts/DBNO_Revive/wav/DBNO_Revive_PlayerJoin_01}
 }
]
properties {
 attenuationSettings=&/DBNO_Revive_Att
 bOverrideConcurrency=true
 maxDistance=768
 soundClassObject=&Fort_SFX_InGame
 volumeMultiplier=1
 concurrencyOverrides {bLimitToOwner=true, maxCount=3, retriggerTime=0.2, voiceStealReleaseTime=0.1}
}
